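# Niagara effect for the Capital emote that throws mesh coins and V-Bucks, launched along rotation sampled from a skeletal mesh.
# Coins burst in a few at a time in partition-based tints and sizes, with age-gated collision that stops their tumble and reorients them on impact.
# Vbucks burst and stream continuously in five colors, inherit mesh velocity and flutter through aerodynamic drag and lift curves.
# Both fall under gravity and shrink away late in life.
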NiagaraSystem /BRCosmetics/Athena/Effects/Particles/Emotes/NS_Emote_Capital/NS_Emote_Capital
dependencies {
 usedDataInterfaces [
  &NiagaraDataInterfaceArrayColor
  &NiagaraDataInterfaceCurve
  &NiagaraDataInterfaceMeshRendererInfo
  &NiagaraDataInterfaceVectorCurve
  &NiagaraDataInterfaceSkeletalMesh
 ]
 usedDynamicInputs [
  &Multiply_Vector2D
  &MakeVector2D
  &RotateVectorByQuaternion
 ]
 usedModules [
  &self:SystemSpawnScript.SystemScriptSource.SystemScriptGraph.NiagaraNodeAssignment_0.SetVariables_#1
  &SystemState
 ]
}
emitters {
 Coins {
  emitterUpdateScript [
   &EmitterState()
   &SpawnBurst_Instantaneous(Spawn Count=&RandomRangeInt(Minimum=4, Maximum=8), Use Spawn Probability=false, UseLoopCountLimit=true)
  ]
  particleSpawnScript [
   &PartitionParticles()
   &InitializeParticle(
    Color=&LerpLinearColors(
     StartColor=<rgba 0.2396 0.1409 0.0721 1>
     EndColor=<rgba 1 1 1 1>
     LerpFactor=&MakeFloatFromInt(Int=$Particles.PartitionParticles.Partition)
    )
    Lifetime=1.5
    Mesh Scale Mode=::NewEnumerator3
    Mesh Uniform Scale=&Lerp_Float(A=0.025, B=0.035, Alpha=&MakeFloatFromInt(Int=$Particles.PartitionParticles.Partition))
    UsePositionOffset=false
    Write Material Random=true
   )
   &ShapeLocation(MaskSpawnGroup=false, Override Seed=false, Sphere Radius=10, UseOwnerScale=true)
   &InitialMeshOrientation(Use Rotation Vector=true)
   &SkeletalMeshLocation(
    Bone Sampling Mode=::NewEnumerator4
    MaskSpawnGroup=false
    Orientation Sampling=::NewEnumerator1
    Skeletal Mesh=$System.Skeletal Mesh
    Write To Particle Position=false
   )
   &InheritVelocity(
    Inherited Velocity Speed Limit=50
    Source Speed Threshold=5
    Speed Limit=true
    UseSpeedThreshold=true
    Velocity Source=$Particles.SkeletalMeshLocation.SampledVelocity
   ) {enabled=false}
   &AddVelocity(
    Additional Quaternion Rotation=false
    Cone Axis=<0 1 0.5>
    Module.UseConeFalloff=false
    Module.UseVelDistribution=false
    Override Seed=false
    Rotation Mode=::NewEnumerator3
    Rotation Quaternion:Quat4f=$Particles.SkeletalMeshLocation.SampledRotation
    Velocity Mode=::NewEnumerator2
    Velocity Speed=&RandomRangeFloat(Minimum=0, Maximum=100)
   )
   &AddVelocity(Velocity=$Particles.ShapeLocation.ShapeNormal, Velocity Speed Scale=100) {module=AddVelocity001}
   &ApplyInitialForces(Apply Force to Position=true)
  ]
  particleUpdateScript [
   &ParticleState()
   &ScaleMeshSize(Scale Factor=&VectorFromFloat(Value=&FloatFromCurve(FloatCurve=DataInterface() {propertyValues {Curve {keys [{tangentMode=RCTM_User, time=0.7, value=1}, {tangentMode=RCTM_User, time=1, value=0}]}}})))
   &GravityForce()
   &Drag(Use Linear Drag=true, Use Rotational Drag=true)
   &Collision(
    Age Colliding Particles=false
    Collision Enabled=&SetBoolByFloatComparison(A=$Particles.NormalizedAge, B=0.25)
    Control Roll On Collision=true
    Module.Control Rest State Via Surface Normal=false
    Module.EnableMaxCollisionCount=false
    Module.Randomize Collision Normal Vector=false
    Restitution=0.2
    ScaleRateOfRotation=false
   )
   &SolveForcesAndVelocity(Clamp Velocity=false, Limit Acceleration=false)
   &UpdateMeshOrientation(
    Rotation Rate=&Multiply_Float(A=45, B=&MakeCustomFloatFromBool(Bool=$Particles.HasCollided, True Float=0, False Float=1))
    Rotation Vector=&RandomRangeVector(Minimum=<0 0 0>, Maximum=<1 1 1>)
   )
   &UpdateMeshOrientation(Enabled:NiagaraBool=$Particles.HasCollided, Orientation Method=::NewEnumerator3, Upright Roll Correction Speed=50) {module=UpdateMeshOrientation001}
  ]
  rendererClasses [&NiagaraMeshRendererProperties]
  renderers [{rendererClass=&NiagaraMeshRendererProperties, rendererIndex=0}]
 }
 Vbucks {
  emitterUpdateScript [
   &EmitterState()
   &SpawnBurst_Instantaneous(Spawn Count=5, Use Spawn Probability=false, UseLoopCountLimit=false)
   &SpawnRate(SpawnRate=60, Use Spawn Probability=false)
  ]
  particleSpawnScript [
   &InitializeParticle(
    Color=&SelectLinearColorFromArray(
     Array Sampling Mode=::NewEnumerator0
     Direct Array Index=?
     Color Selection Array=DataInterface() {propertyValues {ColorData [<rgba 0.9 0.81 0.225 1>, <rgba 1 0.45 0.5875 1>, <rgba 0.36 0.9 0.5535 1>, <rgba 0.8019 0.5 1 1>, <rgba 1 0.5127 0.3 1>]}}
     Random Seed=0
    )
    Lifetime=2.5
    Mass Mode=::NewEnumerator3
    Mesh Renderer Array Visibility Mode=::NewEnumerator3
    Mesh Renderer Info=DataInterface() {propertyValues {MeshRenderer=&self:Vbucks_2.NiagaraMeshRendererProperties_1}}
    Mesh Scale Mode=::NewEnumerator3
    UsePositionOffset=false
    Write Material Random=true
   )
   &ShapeLocation(MaskSpawnGroup=false, Override Seed=false, Sphere Radius=2, UseOwnerScale=true)
   &InitialMeshOrientation(Use Rotation Vector=true)
   &SkeletalMeshLocation(
    Bone Sampling Mode=::NewEnumerator4
    MaskSpawnGroup=false
    Orientation Sampling=::NewEnumerator1
    Skeletal Mesh=$System.Skeletal Mesh
    Write To Particle Position=false
   )
   &InheritVelocity(
    Inherited Velocity Speed Limit=200
    Speed Limit=true
    UseSpeedThreshold=false
    Velocity Source=$Particles.SkeletalMeshLocation.SampledVelocity
   )
   &AddVelocity(
    Additional Quaternion Rotation=false
    Rotation Mode=::NewEnumerator3
    Rotation Quaternion:Quat4f=$Particles.SkeletalMeshLocation.SampledRotation
    Velocity=<0 100 200>
   )
   &AddVelocity(
    Additional Quaternion Rotation=false
    Cone Angle=2
    Cone Axis=<0 2 1>
    Inner Cone Angle=1
    Module.UseConeFalloff=false
    Module.UseVelDistribution=false
    Override Seed=false
    Rotation Mode=::NewEnumerator3
    Rotation Quaternion:Quat4f=$Particles.SkeletalMeshLocation.SampledRotation
    Velocity Mode=::NewEnumerator2
    Velocity Speed=100
   ) {module=AddVelocity001}
   &ApplyInitialForces(Apply Force to Position=true)
  ]
  particleUpdateScript [
   &ParticleState()
   &ScaleMeshSize(Scale Factor=&Multiply_VectorByFloat(
     Vector=&VectorFromCurve(
      VectorCurve=DataInterface() {propertyValues {XCurve {keys [<at 0 1>,<at 1 1>]}, YCurve {keys [<at 0 1>,<at 1 1>]}, ZCurve {keys [<at 0 0.1>,<at 0.4 1>]}}}
      CurveIndex=$Particles.Age
     )
     Float=&FloatFromCurve(FloatCurve=DataInterface() {propertyValues {Curve {keys [{tangentMode=RCTM_User, time=0.7, value=1}, {tangentMode=RCTM_User, time=1, value=0}]}}})
    ))
   &UpdateMeshOrientation(
    Rotation Rate=&RandomRangeFloat(Minimum=-1, Maximum=1)
    Rotation Vector=&Multiply_VectorByFloat(
     Vector=&RandomRangeVector(Minimum=<0 0 0>, Maximum=<1 1 1>)
     Float=&FloatFromCurve(FloatCurve=DataInterface() {propertyValues {Curve {keys [<at 0 0 RCIM_Cubic>,<at 0.1 1 RCIM_Cubic>]}}})
    )
   )
   /Niagara/Modules/Update/Forces/CurlNoiseForce(Noise Frequency=1, Noise Quality / Cost=::NewEnumerator0, Noise Strength=500, Pan Noise=true, Pan Noise Field=<0 0 1>) {enabled=false}
   &AerodynamicDrag(
    Aerodynamic Drag:NiagaraFloat=&FloatFromCurve(
     FloatCurve=DataInterface() {propertyValues {Curve {keys [<at 0 0 RCIM_Cubic>,<at 0.5 1.25 RCIM_Cubic>]}}}
     Scale Curve=&RandomRangeFloat(Minimum=1, Maximum=1.5)
    )
    Aerodynamic Rotational Drag:NiagaraFloat=&RandomRangeFloat(Minimum=0.75, Maximum=1.5)
    Drag Shape Curve=DataInterface() {propertyValues {Curve {keys [
        {interpMode=RCIM_Cubic, tangentMode=RCTM_User, time=0, value=0.1}
        {arriveTangent=2.0799, interpMode=RCIM_Cubic, leaveTangent=2.0799, tangentMode=RCTM_User, time=1, value=1}
       ]}}}
    Ignore Mass=true
    Lift Shape Curve=DataInterface() {propertyValues {Curve {keys [
        {interpMode=RCIM_Cubic, tangentMode=RCTM_User, time=0, value=0}
        {interpMode=RCIM_Cubic, tangentMode=RCTM_User, time=0.5, value=1}
        {tangentMode=RCTM_User, time=1, value=0}
       ]}}}
    Surface Normal Axis:Vector3f=<0 0 1>
    Use Linear Drag=true
    Use Rotational Drag=true
   )
   &GravityForce()
   &SolveForcesAndVelocity(Clamp Velocity=false, Limit Acceleration=false)
   &AlignSpriteToMeshOrientation(SetAlignment=true, SetFacing=true)
  ]
  rendererClasses [&NiagaraMeshRendererProperties]
  renderers [{rendererClass=&NiagaraMeshRendererProperties, rendererIndex=0}]
 }
}
summary {
 systemName=NS_Emote_Capital
 emitters [
  {emitterName=Vbucks, rendererClasses [&NiagaraMeshRendererProperties]}
  {emitterName=Coins, rendererClasses [&NiagaraMeshRendererProperties]}
 ]
 userVariables []
}
