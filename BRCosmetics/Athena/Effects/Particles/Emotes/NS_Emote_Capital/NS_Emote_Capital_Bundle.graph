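# Niagara system for the Capital emote that throws a V-Bucks bundle shedding loose, fluttering mesh pieces.
# Vbucks_Bundle001 bursts a tumbling mesh that inherits capped velocity, falls under gravity and shrinks away.
# Vbucks002 trails it via particle attribute reading at a decaying spawn rate, while Vbucks003 adds an initial burst.
# Both loose-piece emitters pick from five colors and combine curl noise, gravity and orientation-based aerodynamic drag.
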NiagaraSystem /BRCosmetics/Athena/Effects/Particles/Emotes/NS_Emote_Capital/NS_Emote_Capital_Bundle
dependencies {
 usedDataInterfaces [
  &NiagaraDataInterfaceParticleRead
  &NiagaraDataInterfaceCurve
  &NiagaraDataInterfaceArrayColor
  &NiagaraDataInterfaceMeshRendererInfo
  &NiagaraDataInterfaceVectorCurve
  &NiagaraDataInterfaceSkeletalMesh
 ]
 usedDynamicInputs [&Multiply_Vector2D, &MakeVector2D]
 usedModules [
  &self:SystemSpawnScript.SystemScriptSource.SystemScriptGraph.NiagaraNodeAssignment_0.SetVariables_#1
  &SystemState
 ]
}
emitters {
 Vbucks002 {
  emitterUpdateScript [
   &EmitterState()
   &SpawnParticlesFromOtherEmitter(
    Attribute Reader=DataInterface() {propertyValues {EmitterBinding {bindingMode=Other, emitterName=Vbucks_Bundle001}}}
    Maximum Age Spawn Threshold=0.75
    Spawn Rate=&FloatFromCurve(
     FloatCurve=DataInterface() {propertyValues {Curve {keys [
         {arriveTangent=-1.8688, interpMode=RCIM_Cubic, time=0, value=1}
         {arriveTangent=-1.71, interpMode=RCIM_Cubic, leaveTangent=-1.71, tangentMode=RCTM_User, time=1, value=0.1}
        ]}}}
     CurveIndex=$Emitter.NormalizedLoopAge
     Scale Curve=35
    )
    Use Maximum Age=true
    Use Minimum Age=false
   )
   &SpawnBurst_Instantaneous(
    Spawn Count=&RandomRangeInt(Minimum=14, Maximum=18)
    Spawn Time=0.4
    Use Spawn Probability=false
    UseLoopCountLimit=false
   ) {enabled=false}
   &SpawnRate(Spawn Probability=0.7, SpawnRate=80, Use Spawn Probability=true) {enabled=false}
  ]
  particleSpawnScript [
   &InitializeParticle(
    Color=&SelectLinearColorFromArray(
     Array Sampling Mode=::NewEnumerator0
     Direct Array Index=?
     Color Selection Array=DataInterface() {propertyValues {ColorData [<rgba 0.9 0.81 0.225 1>, <rgba 1 0.45 0.5875 1>, <rgba 0.36 0.9 0.5535 1>, <rgba 0.8019 0.5 1 1>, <rgba 1 0.5127 0.3 1>]}}
     Random Seed=0
    )
    Lifetime=1.75
    Mass Mode=::NewEnumerator3
    Mesh Renderer Array Visibility Mode=::NewEnumerator3
    Mesh Renderer Info=DataInterface() {propertyValues {MeshRenderer=&self:Vbucks002_5.NiagaraMeshRendererProperties_1}}
    Mesh Scale Mode=::NewEnumerator3
    UsePositionOffset=false
    Write Material Random=true
   )
   &SampleParticlesFromOtherEmitter(Mesh Orientation Sampling:ENiagara_DirectReadSamplingMode=::NewEnumerator1, Velocity Sampling=::NewEnumerator0)
   &InitialMeshOrientation(Use Rotation Vector=true)
   &SkeletalMeshLocation(Bone Sampling Mode=::NewEnumerator4, MaskSpawnGroup=false, Skeletal Mesh=$System.Skeletal Mesh) {enabled=false}
   &ShapeLocation(MaskSpawnGroup=false, Override Seed=false, Sphere Radius=1, UseOwnerScale=true)
   &InheritVelocity(
    Inherited Velocity Speed Limit:NiagaraFloat=&FloatFromCurve(
     FloatCurve=DataInterface() {propertyValues {Curve {keys [
         {arriveTangent=4.1785, interpMode=RCIM_Cubic, leaveTangent=4.1785, tangentMode=RCTM_User, time=0, value=0}
         {arriveTangent=0.0224, interpMode=RCIM_Cubic, leaveTangent=0.0224, tangentMode=RCTM_User, time=1, value=1}
        ]}}}
     CurveIndex=$Emitter.NormalizedLoopAge
     Scale Curve=100
    )
    Speed Limit=true
    UseSpeedThreshold=false
    Velocity Source=$Output.SampleParticlesFromOtherEmitter.SampledVelocity
   )
   &AddVelocity(
    Module.Use Falloff Distance=false
    Override Seed=false
    Velocity Mode=::NewEnumerator1
    Velocity Speed=&Multiply_Float(
     A=&RandomRangeFloat(Minimum=30, Maximum=60)
     B=&FloatFromCurve(FloatCurve=DataInterface() {propertyValues {Curve {keys [<at 0.2 0>,<at 0.35 1>]}}}, CurveIndex=$Emitter.LoopedAge)
    )
   ) {enabled=false}
   &ApplyInitialForces()
  ]
  particleUpdateScript [
   &ParticleState()
   &ScaleMeshSize(Scale Factor=&Multiply_VectorByFloat(
     Vector=&VectorFromCurve(
      VectorCurve=DataInterface() {propertyValues {XCurve {keys [<at 0 1>,<at 1 1>]}, YCurve {keys [<at 0 1>,<at 1 1>]}, ZCurve {keys [<at 0 0.1>,<at 0.6 1>]}}}
      CurveIndex=$Particles.Age
     )
     Float=&FloatFromCurve(FloatCurve=DataInterface() {propertyValues {Curve {keys [{tangentMode=RCTM_User, time=0.7, value=1}, {tangentMode=RCTM_User, time=1, value=0}]}}})
    ))
   &UpdateMeshOrientation(
    Rotation Rate=&RandomRangeFloat(Minimum=-1, Maximum=1)
    Rotation Vector=&Multiply_VectorByFloat(
     Vector=&RandomRangeVector(Minimum=<0 0 0>, Maximum=<1 1 1>)
     Float=&FloatFromCurve(FloatCurve=DataInterface() {propertyValues {Curve {keys [<at 0 0 RCIM_Cubic>,<at 0.1 1 RCIM_Cubic>]}}})
    )
   )
   /Niagara/Modules/Update/Forces/CurlNoiseForce(Noise Frequency=4, Noise Quality / Cost=::NewEnumerator0, Noise Strength=250, Pan Noise=true, Pan Noise Field=<0 0 1>)
   &GravityForce()
   &AerodynamicDrag(
    Aerodynamic Drag:NiagaraFloat=&FloatFromCurve(
     FloatCurve=DataInterface() {propertyValues {Curve {keys [{tangentMode=RCTM_User, time=0, value=0.35}, {tangentMode=RCTM_User, time=0.5, value=1}]}}}
     Scale Curve=&RandomRangeFloat(Minimum=1, Maximum=1.5)
    )
    Aerodynamic Rotational Drag:NiagaraFloat=&RandomRangeFloat(Minimum=0.75, Maximum=1.5)
    Ignore Mass=true
    Surface Normal Axis:Vector3f=<0 0 1>
    Use Linear Drag=true
    Use Rotational Drag=true
   )
   &SolveForcesAndVelocity(Clamp Velocity=false, Limit Acceleration=false)
   &AlignSpriteToMeshOrientation(SetAlignment=true, SetFacing=true)
  ]
  rendererClasses [&NiagaraMeshRendererProperties]
  renderers [{rendererClass=&NiagaraMeshRendererProperties, rendererIndex=0}]
 }
 Vbucks003 {
  emitterUpdateScript [
   &EmitterState()
   &SpawnBurst_Instantaneous(
    Spawn Count=&RandomRangeInt(Minimum=12, Maximum=18)
    Spawn Time=0.01
    Use Spawn Probability=false
    UseLoopCountLimit=false
   )
  ]
  particleSpawnScript [
   &InitializeParticle(
    Color=&SelectLinearColorFromArray(
     Array Sampling Mode=::NewEnumerator0
     Direct Array Index=?
     Color Selection Array=DataInterface() {propertyValues {ColorData [<rgba 0.9 0.81 0.225 1>, <rgba 1 0.45 0.5875 1>, <rgba 0.36 0.9 0.5535 1>, <rgba 0.8019 0.5 1 1>, <rgba 1 0.5127 0.3 1>]}}
     Random Seed=0
    )
    Lifetime=1.75
    Mass Mode=::NewEnumerator3
    Mesh Renderer Array Visibility Mode=::NewEnumerator3
    Mesh Renderer Info=DataInterface() {propertyValues {MeshRenderer=&self:Vbucks003_1.NiagaraMeshRendererProperties_1}}
    Mesh Scale Mode=::NewEnumerator3
    UsePositionOffset=false
    Write Material Random=true
   )
   &InitialMeshOrientation(Use Rotation Vector=true)
   &SkeletalMeshLocation(Bone Sampling Mode=::NewEnumerator4, MaskSpawnGroup=false, Skeletal Mesh=$System.Skeletal Mesh) {enabled=false}
   &ShapeLocation(MaskSpawnGroup=false, Override Seed=false, Sphere Radius=2, UseOwnerScale=true)
   &InheritVelocity(Inherited Velocity Amount Scale=<1 1 1.5>, Inherited Velocity Speed Limit=450, Speed Limit=true, UseSpeedThreshold=false)
   &AddVelocity(
    Module.Use Falloff Distance=false
    Override Seed=false
    Velocity Mode=::NewEnumerator1
    Velocity Speed=&RandomRangeFloat(Minimum=65, Maximum=100)
   )
   &ApplyInitialForces()
  ]
  particleUpdateScript [
   &ParticleState()
   &ScaleMeshSize(Scale Factor=&Multiply_VectorByFloat(
     Vector=&VectorFromCurve(
      VectorCurve=DataInterface() {propertyValues {XCurve {keys [<at 0 1>,<at 1 1>]}, YCurve {keys [<at 0 1>,<at 1 1>]}, ZCurve {keys [<at 0 0.1>,<at 0.6 1>]}}}
      CurveIndex=$Particles.Age
     )
     Float=&FloatFromCurve(FloatCurve=DataInterface() {propertyValues {Curve {keys [{tangentMode=RCTM_User, time=0.7, value=1}, {tangentMode=RCTM_User, time=1, value=0}]}}})
    ))
   &UpdateMeshOrientation(
    Rotation Rate=&RandomRangeFloat(Minimum=-1, Maximum=1)
    Rotation Vector=&Multiply_VectorByFloat(
     Vector=&RandomRangeVector(Minimum=<0 0 0>, Maximum=<1 1 1>)
     Float=&FloatFromCurve(FloatCurve=DataInterface() {propertyValues {Curve {keys [<at 0 0 RCIM_Cubic>,<at 0.1 1 RCIM_Cubic>]}}})
    )
   )
   &CurlNoiseForce(Noise Frequency=4, Noise Quality / Cost=::NewEnumerator0, Noise Strength=250, Pan Noise=true, Pan Noise Field=<0 0 1>)
   &GravityForce()
   &AerodynamicDrag(
    Aerodynamic Drag:NiagaraFloat=&FloatFromCurve(
     FloatCurve=DataInterface() {propertyValues {Curve {keys [{tangentMode=RCTM_User, time=0, value=0.35}, {tangentMode=RCTM_User, time=0.5, value=1}]}}}
     Scale Curve=&RandomRangeFloat(Minimum=1, Maximum=1.5)
    )
    Aerodynamic Rotational Drag:NiagaraFloat=&RandomRangeFloat(Minimum=0.75, Maximum=1.5)
    Ignore Mass=true
    Surface Normal Axis:Vector3f=<0 0 1>
    Use Linear Drag=true
    Use Rotational Drag=true
   )
   &SolveForcesAndVelocity(Clamp Velocity=false, Limit Acceleration=false)
   &AlignSpriteToMeshOrientation(SetAlignment=true, SetFacing=true)
  ]
  rendererClasses [&NiagaraMeshRendererProperties]
  renderers [{rendererClass=&NiagaraMeshRendererProperties, rendererIndex=0}]
 }
 Vbucks_Bundle001 {
  emitterUpdateScript [
   &EmitterState()
   &SpawnBurst_Instantaneous(Spawn Time=0.05, Use Spawn Probability=false, UseLoopCountLimit=true)
  ]
  particleSpawnScript [
   &InitializeParticle(Lifetime=0.7, Mesh Scale Mode=::NewEnumerator3, Mesh Uniform Scale=1.2, UsePositionOffset=false, Write Material Random=true)
   &SkeletalMeshLocation(Bone Sampling Mode=::NewEnumerator4, MaskSpawnGroup=false, Orientation Sampling=::NewEnumerator1, Skeletal Mesh=$System.Skeletal Mesh) {enabled=false}
   &InheritVelocity(Inherited Velocity Speed Limit=450, Speed Limit=true, UseSpeedThreshold=false)
   &ApplyInitialForces()
   &InitialMeshOrientation(Use Rotation Vector=false)
  ]
  particleUpdateScript [
   &ParticleState()
   &ScaleMeshSize(Scale Factor=&VectorFromFloat(Value=&FloatFromCurve(FloatCurve=DataInterface() {propertyValues {Curve {keys [
          {tangentMode=RCTM_User, time=0, value=1}
          {arriveTangent=-0.0405, interpMode=RCIM_Cubic, leaveTangent=-0.0405, tangentMode=RCTM_User, time=0.7009, value=1}
          {arriveTangent=-8.0089, interpMode=RCIM_Cubic, leaveTangent=-8.0089, tangentMode=RCTM_User, time=1, value=0}
         ]}}})))
   &GravityForce(Gravity=<0 0 -850>)
   &Drag(Use Linear Drag=true, Use Rotational Drag=true)
   &Collision(
    Age Colliding Particles=false
    Collision Enabled=&SetBoolByFloatComparison(A=$Particles.NormalizedAge, B=0.05)
    Control Roll On Collision=true
    Friction=1
    Module.Control Rest State Via Surface Normal=false
    Module.EnableMaxCollisionCount=false
    Module.Randomize Collision Normal Vector=false
    Particle Radius Scale=3
    Radius Calculation Type=::NewEnumerator1
    Restitution=0.025
    ScaleRateOfRotation=false
   ) {enabled=false}
   &SolveForcesAndVelocity(Clamp Velocity=false, Limit Acceleration=false)
   &UpdateMeshOrientation(
    Rotation Rate=&Multiply_Float(A=1, B=&MakeCustomFloatFromBool(Bool=$Particles.HasCollided, True Float=0, False Float=1))
    Rotation Vector=&RandomRangeVector(Minimum=<-1 -1 -1>, Maximum=<1 1 1>)
   )
   &UpdateMeshOrientation(
    Bank Rate=1
    Enabled:NiagaraBool=$Particles.HasCollided
    "Foward Vector (Mesh Space)":Vector3f=&RandomRangeVector(Minimum=<0 -1 0>, Maximum=<0 1 0>)
    Orientation Method=::NewEnumerator3
    "Right Vector (Mesh Space)":Vector3f=&RandomRangeVector(Minimum=<-1 -1 0>, Maximum=<1 1 0>)
    Up Vector:Vector3f=&RandomRangeVector(Minimum=<-1 0 0>, Maximum=<1 0 0>)
    Upright Roll Correction Speed=4
    Velocity Orientation Convergence Rate=50
   ) {enabled=false, module=UpdateMeshOrientation001}
  ]
  rendererClasses [&NiagaraMeshRendererProperties]
  renderers [{rendererClass=&NiagaraMeshRendererProperties, rendererIndex=0}]
 }
}
summary {
 systemName=NS_Emote_Capital_Bundle
 emitters [
  {emitterName=Vbucks002, rendererClasses [&NiagaraMeshRendererProperties]}
  {emitterName=Vbucks_Bundle001, rendererClasses [&NiagaraMeshRendererProperties]}
  {emitterName=Vbucks003, rendererClasses [&NiagaraMeshRendererProperties]}
 ]
 userVariables []
}
